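# Material function for the Flashy Fever shoe that adds audio-reactive color, glow and faceted normals to input material attributes.
# An FX mask texture supplies effect and facet masks, while an ID map provides grayscale IDs for per-facet variation.
# Time, offset by pre-skinned X position, and frontend-loop audio amplitude with a frontend cycle feed the animation.
# Sine and frac patterns drive a four-color emissive gradient and desaturated base tint, blended over the input.
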MaterialFunction /CosmeticShoes/Assets/FORT_Shoes/Shoe_FlashyFever/Materials/MF_FlashyFever_Shoe
graph {nodes [
  Constant()
  n1 = Power(Base=NamedRerouteUsage(), Exp=ScalarParameter() {defaultValue=0.5, group=FlashyFever, parameterName=Exp_1}) {constExponent=4}
  n2 = Multiply(A=Subtract(A=NamedRerouteUsage()) {constB=0.1535}) {constB=1.5}
  n3 = NamedRerouteUsage()
  n4 = Add(A=NamedRerouteUsage()) {constB=0.3534}
  n5 = Add(A=NamedRerouteUsage(), B=Add(A=NamedRerouteUsage()) {constB=-0.1552})
  n6 = Sine(None=Min(A=Subtract(A=@n3, B=@n4), B=Multiply(A=@n5) {constB=0.5}))
  n7 = Frac(None=Subtract(A=Add(A=@n1, B=@n2), B=@n6) {constB=0})
  Divide(A=Round(None=Multiply(A=@n7) {constB=4})) {constB=4}
  n8 = NamedRerouteUsage()
  n9 = NamedRerouteUsage()
  n10 = GetMaterialAttributes(None=FunctionInput() {inputName=Material, inputType=FunctionInput_MaterialAttributes}) {attributeGetTypes [#1, #2, #3]}
  n11 = NamedRerouteDeclaration(None=@n10.MaterialAttributes) {name=InMaterial, nodeColor=<rgba 0 0.9297 1 1>}
  n12 = NamedRerouteUsage()
  n13 = NamedRerouteUsage()
  n14 = NamedRerouteUsage()
  n15 = NamedRerouteUsage()
  FunctionOutput(None=BlendMaterialAttributes(
    A=SetMaterialAttributes!(MaterialAttributes=@n8, Normal=@n9) {attributeSetTypes [#3]}
    B=SetMaterialAttributes!(MaterialAttributes=@n11, BaseColor=@n12, EmissiveColor=@n13) {attributeSetTypes [#1, #2]}
    Alpha=Multiply(A=@n14, B=@n15)
   ))
  NamedRerouteDeclaration(None=ConstantBiasScale(None=NamedRerouteUsage()) {bias=-0.5, scale=2}) {name=OffsettedID, nodeColor=<rgba 0.125 1 0 1>}
  n16 = MaterialFunctionCall(Color=FunctionInput(Preview=TextureSample() {samplerType=SAMPLERTYPE_Masks, texture=/CosmeticShoes/Assets/FORT_Shoes/Shoe_FlashyFever/Textures/T_FlashyFever_Shoe_FX}.RGB) {inputName=FXMask}) {materialFunction=&SplitComponents}
  NamedRerouteDeclaration(None=SmoothStep(Value=@n16.B) {constMin=0.9}) {name=FacetsMask, nodeColor=<rgba 1 0 0.7031 1>}
  NamedRerouteDeclaration(None=@n16.R) {name=EffectMask, nodeColor=<rgba 0 1 0.9766 1>}
  NamedRerouteDeclaration(None=TextureSample() {samplerType=SAMPLERTYPE_LinearColor, texture=/CosmeticShoes/Assets/FORT_Shoes/Shoe_FlashyFever/Textures/T_FlashyFever_Shoe_IDMap}.G) {name=GrayscaleIDMap, nodeColor=<rgba 1 0.6563 0 1>}
  NamedRerouteDeclaration(None=@n10.BaseColor) {name=InBaseColor, nodeColor=<rgba 1 0 0.8672 1>}
  n17 = NamedRerouteUsage()
  n18 = Desaturation(None=@n17, Fraction=Constant() {r=1})
  n19 = VectorParameter() {group=FlashyFever, parameterName=ReactivityBaseColor, defaultValue=<rgba 0.04 0 0.0833 0.8609>}
  n20 = NamedRerouteUsage()
  n21 = Multiply(A=NamedRerouteUsage()) {constB=0.25}
  n22 = Subtract(A=OneMinus(None=Add(A=@n20, B=@n21))) {constB=0.25}
  n23 = Multiply(A=NamedRerouteUsage(), B=Power(Base=Frac(None=@n22)) {constExponent=4})
  NamedRerouteDeclaration(None=LinearInterpolate(A=@n17, B=Multiply(A=@n18, B=@n19.RGB), Alpha=LinearInterpolate(A=@n19.A, Alpha=@n23) {constA=0.9})) {name=OutBaseColor, nodeColor=<rgba 0 1 0.8125 1>}
  n24 = ConstantBiasScale(None=NamedRerouteUsage()) {bias=-0.5, scale=2}
  n25 = ScalarParameter() {defaultValue=0.1, group=FlashyFever, parameterName=FacetsNormalRandomness}
  n26 = Add(A=NamedRerouteUsage(), B=Multiply(A=Multiply(A=@n24, B=@n25), B=Constant3Vector() {constant=<rgba 1 1 0 0>}))
  NamedRerouteDeclaration(None=LinearInterpolate(A=NamedRerouteUsage(), B=Normalize(VectorInput=@n26), Alpha=NamedRerouteUsage())) {name=OutNormals, nodeColor=<rgba 0 1 0.25 1>}
  NamedRerouteDeclaration(None=@n10.EmissiveColor) {name=InEmissive, nodeColor=<rgba 0 0.1094 1 1>}
  n27 = VectorParameter() {group=FlashyFever, parameterName=ReactivityColor1, defaultValue=<rgba 1 0.9405 0.0936 1>}
  n28 = VectorParameter() {group=FlashyFever, parameterName=ReactivityColor2, defaultValue=<rgba 1 0 0.8318 1>}
  n29 = VectorParameter() {group=FlashyFever, parameterName=ReactivityColor3, defaultValue=<rgba 0 0.8193 1 1>}
  n30 = VectorParameter() {group=FlashyFever, parameterName=ReactivityColor4, defaultValue=<rgba 0.0452 1 0 1>}
  n31 = ScalarParameter() {group=FlashyFever, parameterName=EmissiveSaturation}
  n32 = Multiply(
   A=ScalarParameter() {defaultValue=1.5, group=FlashyFever, parameterName=EmissiveIntensityMultiplier}
   B=Desaturation(
    None=MaterialFunctionCall!(A=@n27.RGB, B=@n28.RGB, C=@n29.RGB, D=@n30.RGB, Alpha=@n7) {materialFunction=/BRCosmetics/Characters/Player/Female/Medium/Bodies/F_MED_PolarityWinn_Field/Materials/MaterialFunctions/Lerp_4Color}.Result
    Fraction=Multiply(A=@n31) {constB=-1}
   )
  )
  n33 = Add(A=NamedRerouteUsage(), B=Add(A=NamedRerouteUsage()) {constB=-0.1552})
  n34 = Multiply(A=@n6, B=Sine(None=Multiply(A=@n33) {constB=1.5}))
  n35 = Multiply(A=NamedRerouteUsage(), B=SphereMask(A=Saturate(None=@n34), B=Constant() {r=1}) {attenuationRadius=0.5, hardnessPercent=80})
  NamedRerouteDeclaration(None=Add(A=NamedRerouteUsage(), B=Multiply(A=@n32, B=@n35))) {name=OutEmissive, nodeColor=<rgba 0.4297 1 0 1>}
  n36 = Multiply(
   A=Power(
    Base=MaterialFunctionCall!(InAmplitude, InSpectrum, CustomFakeSpectrum, CustomFakeAmplitude) {materialFunction=/BRCosmetics/Global/MaterialFunctions/AudioReactivity/MF_AudioReactivity_FrontendLoop}.Amplitude
    Exp=ScalarParameter() {defaultValue=0.2, group=FlashyFever, parameterName=Exp}
   )
   B=ScalarParameter() {defaultValue=1, group=FlashyFever, parameterName=EnableReactivity}
  )
  n37 = NamedRerouteUsage()
  NamedRerouteDeclaration(None=LinearInterpolate(A=@n36, B=Multiply(A=@n36, B=@n37), Alpha=ScalarParameter() {defaultValue=1, group=FlashyFever, parameterName=IsInFrontend})) {name=InAmplitude, nodeColor=<rgba 0 0.0156 1 1>}
  n38 = Time()
  n39 = ScalarParameter() {defaultValue=0.3, group=FlashyFever, parameterName=TimeSpeed}
  n40 = VertexInterpolator(VS=LocalPosition() {localOrigin=InstancePreSkinning}.XYZ)
  n41 = Saturate(None=ComponentMask(None=NamedRerouteDeclaration(None=@n40.PS) {name=PreSkinnedPosition, nodeColor=<rgba 1 0.8438 0 1>}) {r=true})
  NamedRerouteDeclaration(None=Add(A=Multiply(A=@n38, B=@n39) {constB=10}, B=NamedRerouteDeclaration(None=@n41) {name=SideMask, nodeColor=<rgba 0.5234 0 1 1>})) {name=OffsettedTime, nodeColor=<rgba 0.8047 0 1 1>}
  n42 = Sine(None=Multiply(A=Time(), B=ScalarParameter() {defaultValue=0.07, group=FlashyFever, parameterName=OffPawnFakeReactivitySpeed}) {constB=0.1})
  NamedRerouteDeclaration(None=SmoothStep(Value=Saturate(None=@n42)) {constMax=0.2}) {name=FrontEndCycle, nodeColor=<rgba 0.1016 0 1 1>}
  NamedRerouteDeclaration(None=@n10.Normal) {name=InNormal, nodeColor=<rgba 0.8984 0 1 1>}
  NamedRerouteUsage()
 ]}
